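; Function library for the Wild Estate Zero Point Dash, providing GetTotalCharges to report dash charges held by an Ability System Component.
; It reads the CurrentCharges float attribute from the FortChargingSet_SunroseDash attribute set and truncates it to an integer.
; An invalid component or a failed attribute lookup returns zero.
; Despite the function's name, the value returned is the current charge count.
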(fn GetTotalCharges
  (params (AbilitySystemComponent :type Ability%20System%20Component%20Object%20Reference))
  (bind (_bsuccessfullyfoundattribute _returnvalue) (Ability|Attribute|GetFloatAttributefromAbilitySystemComponent AbilitySystemComponent "(AttributeName=\"CurrentCharges\",Attribute=/Script/SunRoseEnvGameplayRuntime.FortChargingSet_SunroseDash:CurrentCharges,AttributeOwner=\"/Script/CoreUObject.Class'/Script/SunRoseEnvGameplayRuntime.FortChargingSet_SunroseDash'\")"))
  (Utilities|IsValid AbilitySystemComponent
    (:Is%20Valid
      (if _bsuccessfullyfoundattribute
        (return (Math|Float|Truncate _returnvalue))
        (else
          _)))
    (:Is%20Not%20Valid
      _))
  (return 0))
